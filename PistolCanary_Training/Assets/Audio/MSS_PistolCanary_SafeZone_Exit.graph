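# Stereo one-shot MetaSound source for exiting a safe zone in PistolCanary Training.
# It is a preset of the MetaSoundWizard randomized-array one-shot, supplying two temporary GrillMoon POI boundary-exit waves as the pool.
# Pitch and volume randomization are zeroed, so variation comes only from wave selection.
# It uses a very-close POI attenuation asset and the Fort gameplay events sound class.
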MetaSoundSource /PistolCanary_Training/Assets/Audio/MSS_PistolCanary_SafeZone_Exit
preset_parent=/MetaSoundWizard/MetaSounds/One-shots/ArrayRandomized/MSS_Oneshot_Array_Randomized_Stereo
inputs [
 Random.Pitch.Base {type=Float}
 Random.Pitch.Spread {type=Float}
 Random.Volume.Base {type=Float}
 Random.Volume.Spread {type=Float}
 Sounds {type=WaveAsset:Array}
 UE.Source.OnPlay {type=Trigger}
]
interfaces [UE.OutputFormat.Stereo:1.0, UE.Source.OneShot:1.0, UE.Source:1.1]
outputs [
 "UE.OutputFormat.Stereo.Audio:0" {type=Audio}
 "UE.OutputFormat.Stereo.Audio:1" {type=Audio}
 UE.Source.OneShot.OnFinished {type=Trigger}
]
pages [Default {input_defaults {
   Random.Pitch.Base=0
   Random.Pitch.Spread=0
   Random.Volume.Base=0
   Random.Volume.Spread=0
   UE.Source.OnPlay=null
   Sounds [
    /PistolCanary_Training/Assets/Audio/Wavs/Temp/GrillMoon_POI_Boundary_Exit_01
    /PistolCanary_Training/Assets/Audio/Wavs/Temp/GrillMoon_POI_Boundary_Exit_02
   ]
  }}]
properties {
 assetImportData=&self:AssetImportData
 attenuationSettings=/PistolCanary_Training/Assets/Audio/Wavs/Temp/ATT_PistolCanary_POI_VeryClose
 maxDistance=4100
 numChannels=2
 outputFormat=Stereo
 sampleRate=48000
 soundClassObject=&Fort_Events_Gameplay
}
